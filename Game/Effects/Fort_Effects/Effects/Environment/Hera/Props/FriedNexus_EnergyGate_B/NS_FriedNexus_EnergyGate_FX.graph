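# Niagara system for the ambient energy effect on the FriedNexus Energy Gate B environment prop, built around a cyan glowing sphere.
# The Sphere emitter draws a bobbing, waveform-pulsed mesh with a light, and SphereLightning adds intermittent, randomly oriented mesh arcs.
# SideLightning spawns center-oriented bolts and sprites at four fixed anchors, driven by dynamic material parameters, while AirglowAndDecals places flickering glow sprites near those points.
# Time-of-day feeds a sphere material parameter.
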NiagaraSystem /Game/Effects/Fort_Effects/Effects/Environment/Hera/Props/FriedNexus_EnergyGate_B/NS_FriedNexus_EnergyGate_FX
dependencies {
 usedDataInterfaces [&NiagaraDataInterfaceArrayFloat3, &NiagaraDataInterfaceCamera, &NiagaraDataInterfaceCurve]
 usedDynamicInputs []
 usedModules [&SystemState]
}
emitters {
 AirglowAndDecals {
  emitterUpdateScript [
   &EmitterState()
   &SpawnBurst_Instantaneous(Spawn Count=4, Use Spawn Probability=false, UseLoopCountLimit=true)
  ]
  particleSpawnScript [
   &PartitionParticles(Number Of Partitions=4)
   &InitializeParticle(
    Color=<rgba 0 0.5982 1 1>
    Lifetime=5
    Position Offset=&SelectVectorFromArray(
     Direct Array Index=$Particles.PartitionParticles.Partition
     Vector Selection Array=DataInterface() {propertyValues {FloatData [<58 -1 17>, <58 -1 -18>, <-66 0 17>, <-66 0 -18>]}}
    )
    Sprite Size Mode=::NewEnumerator3
    Uniform Sprite Size=35
    UsePositionOffset=true
    Write Material Random=true
   )
   &CameraOffset(Camera Offset Amount=5)
  ]
  particleUpdateScript [
   &ParticleState(Kill Particles When Lifetime Has Elapsed=false, Let Infinitely Lived Particles Die When Emitter Deactivates=true)
   &ScaleColor(
    Scale Alpha=&MakeFloatFromVector(
     Vector=&VectorFromFloat(Value=&Waveform(
       Global Input Position=$Particles.PartitionParticles.NormalizedPartition
       "Waveform [1]"=::NewEnumerator12
       "[1] Frequency"=4
       "[1] Amplitude Min/Max"=<0 1>
       "[1] Bias"=0.5
      ))
     Channel=::NewEnumerator0
    )
    ScaleA=true
    ScaleRGB=false
   )
   &ScaleColor(
    Scale Alpha=&FloatFromCurve(
     FloatCurve=DataInterface() {propertyValues {Curve {
        defaultValue=0
        keys [
         {interpMode=RCIM_Cubic, tangentMode=RCTM_User, time=0, value=0.25}
         {interpMode=RCIM_Cubic, tangentMode=RCTM_User, time=0.5, value=1}
         {tangentMode=RCTM_User, time=1, value=0.25}
        ]
       }}}
     CurveIndex=Linked() {linkedVariable {name=NPC.FortniteNPC.TimeOfDay_0-1, type=NiagaraFloat}}
    )
    ScaleA=true
    ScaleRGB=true
   ) {enabled=false, module=ScaleColor001}
   /SetVariables(Particles.Decal Color:LinearColor=&Multiply_LinearColorByFloat(Linear Color=$Particles.Color, Float=3)) {isSetParameters=true}
   /SetVariables(Particles.Decal Sprite Size:Vector2f=&Multiply_Vector2D_ByFloat(Float=0.6, Vector2D=$Particles.SpriteSize)) {isSetParameters=true}
  ]
  rendererClasses [&NiagaraSpriteRendererProperties]
  renderers [
   {rendererClass=&NiagaraSpriteRendererProperties}
   {rendererClass=&NiagaraSpriteRendererProperties, rendererIndex=1}
  ]
 }
 SideLightning {
  emitterUpdateScript [
   &EmitterState(Life Cycle Mode=::NewEnumerator1, Loop Behavior=::NewEnumerator1, Loop Duration Mode=::NewEnumerator1, UseLoopDelay=false)
   &SpawnBurst_Instantaneous(Spawn Count=10, Use Spawn Probability=false, UseLoopCountLimit=false) {enabled=false}
   &SpawnRate(SpawnRate=20, Use Spawn Probability=false)
  ]
  particleSpawnScript [
   &PartitionParticles(Number Of Partitions=4)
   /SetVariables(Particles.GlobalRandom:NiagaraFloat=&Multiply_Float(A=&RandomRangeFloat(Minimum=0.9, Maximum=1), B=0.6)) {isSetParameters=true}
   /SetVariables(Particles.GlobalWidth:NiagaraFloat=&RandomRangeFloat(Minimum=0.5, Maximum=3)) {isSetParameters=true}
   &InitializeParticle(
    Color=<rgba 0 0.6609 1 1>
    Lifetime=&RandomRangeFloat(Minimum=0.25, Maximum=0.75)
    Mesh Scale=&MakeVector(X=1, Y=1, Z=&Multiply_Float(A=1, B=$Particles.GlobalRandom))
    Mesh Scale Mode=::NewEnumerator4
    Sprite Size Mode=::NewEnumerator3
    Uniform Sprite Size=&Multiply_Float(A=1.5, B=$Particles.GlobalWidth)
    UsePositionOffset=false
    Write Material Random=true
   )
   &ShapeLocation(
    MaskSpawnGroup=false
    Non Uniform Scale=<0.5 1 1>
    Offset=&SelectVectorFromArray(
     Direct Array Index=$Particles.PartitionParticles.Partition
     Vector Selection Array=DataInterface() {propertyValues {FloatData [<62 -1 17>, <62 -1 -18>, <-70 0 17>, <-70 0 -18>]}}
     Random Seed=0
    )
    Offset Mode=::NewEnumerator0
    Override Seed=false
    Sphere Radius=9
    UseOwnerScale=true
   )
   /SetVariables(Particles.AimAtCenter:Vector3f=&NormalizeVector(
     Vector To Normalize=Linked() {linkedVariable {name=Particles.Position, type=NiagaraPosition}}
     Safe Normalize=true
     Safe Fallback Threshold=0
     Safe Fallback Vector=<0 0 0.001>
    )) {isSetParameters=true}
   &InitialMeshOrientation(
    Mesh Axis To Orient=<0 0 -1>
    Mesh Orientation Mode=::NewEnumerator2
    Orientation Vector=$Particles.AimAtCenter
    Rotation=&RandomRangeVector(Minimum=<0 0 0>, Maximum=<0 0 1>)
    Use Rotation Vector=true
   )
  ]
  particleUpdateScript [
   &ParticleState()
   &DynamicMaterialParameters(
    Index 0 Param 1=&FloatFromCurve(
     FloatCurve=DataInterface() {propertyValues {Curve {
        defaultValue=0
        keys [
         {interpMode=RCIM_Cubic, tangentMode=RCTM_User, time=0, value=0}
         {arriveTangent=1.4591, interpMode=RCIM_Cubic, leaveTangent=1.4591, time=0.3529, value=0.4369}
         {arriveTangent=5.5065, interpMode=RCIM_Cubic, leaveTangent=5.5065, time=0.9017, value=1.3157}
         {arriveTangent=16.4692, interpMode=RCIM_Cubic, leaveTangent=16.4692, tangentMode=RCTM_User, time=1, value=4}
        ]
       }}}
     Scale Curve=$Particles.GlobalWidth
    )
    Index 0 Param 2=&FloatFromCurve(
     FloatCurve=DataInterface() {propertyValues {Curve {
        defaultValue=1
        keys [
         {arriveTangent=0.0094, interpMode=RCIM_Cubic, leaveTangent=0.0094, tangentMode=RCTM_User, time=0.5, value=0.4}
         {arriveTangent=1.9089, interpMode=RCIM_Cubic, leaveTangent=1.9089, tangentMode=RCTM_User, time=1, value=1}
        ]
       }}}
     Scale Curve=&RandomRangeFloat(Minimum=10, Maximum=12)
    )
    Index 0 Param 3=&RandomRangeFloat(Minimum=0.25, Maximum=1)
    Index 0 Param 4=&FloatFromCurve(FloatCurve=DataInterface() {propertyValues {Curve {
        defaultValue=0
        keys [
         {interpMode=RCIM_Cubic, tangentMode=RCTM_User, time=0, value=0}
         {arriveTangent=2.7163, interpMode=RCIM_Cubic, leaveTangent=2.7163, tangentMode=RCTM_User, time=0.7, value=1}
        ]
       }}})
    Index 1 Param 1=&FloatFromCurve(FloatCurve=DataInterface() {propertyValues {Curve {
        defaultValue=0.0353
        keys [{interpMode=RCIM_Cubic, tangentMode=RCTM_User, time=0.5, value=0}, {interpMode=RCIM_Cubic, tangentMode=RCTM_User, time=1, value=1}]
       }}})
    Index 1 Param 2=&Multiply_Float(A=1.35, B=$Particles.GlobalRandom)
    Index 1 Param 3=&Multiply_Float(A=&RandomRangeFloat(Minimum=0.0125, Maximum=0.025), B=$Particles.GlobalRandom)
    Index 1 Param 4=0.5
    Module.Index0Float0Write=true
    Module.Index0Float1Write=true
    Module.Index0Float2Write=true
    Module.Index0Float3Write=true
    Module.Index1Float0Write=true
    Module.Index1Float1Write=true
    Module.Index1Float2Write=true
    Module.Index1Float3Write=true
    Write Parameter Index 1=::NewEnumerator1
   )
   &ScaleColor(
    Scale Alpha=&FloatFromCurve(
     FloatCurve=DataInterface() {propertyValues {Curve {
        defaultValue=0
        keys [
         {interpMode=RCIM_Cubic, tangentMode=RCTM_User, time=0, value=0.5}
         {interpMode=RCIM_Cubic, tangentMode=RCTM_User, time=0.5, value=1}
         {tangentMode=RCTM_User, time=1, value=0.5}
        ]
       }}}
     CurveIndex=Linked() {linkedVariable {name=NPC.FortniteNPC.TimeOfDay_0-1, type=NiagaraFloat}}
    )
    ScaleA=true
    ScaleRGB=true
   ) {enabled=false, module=ScaleColor001}
   /SetVariables(Particles.SpriteColor=&Multiply_LinearColorByFloat(
     Linear Color=$Particles.Color
     Float=&FloatFromCurve(
      FloatCurve=DataInterface() {propertyValues {Curve {
         defaultValue=0
         keys [
          {interpMode=RCIM_Cubic, tangentMode=RCTM_User, time=0, value=0}
          {interpMode=RCIM_Cubic, tangentMode=RCTM_User, time=0.5, value=1}
          {tangentMode=RCTM_User, time=1, value=0}
         ]
        }}}
      Scale Curve=10
     )
    )) {isSetParameters=true}
  ]
  rendererClasses [&NiagaraMeshRendererProperties, &NiagaraSpriteRendererProperties]
  renderers [
   {rendererClass=&NiagaraMeshRendererProperties, rendererIndex=0}
   {rendererClass=&NiagaraSpriteRendererProperties, rendererIndex=1}
  ]
 }
 Sphere {
  emitterUpdateScript [
   &EmitterState(
    Inactive Response=::NewEnumerator1
    Life Cycle Mode=::NewEnumerator1
    Loop Behavior=::NewEnumerator1
    Loop Duration Mode=::NewEnumerator1
    UseLoopDelay=false
   )
   &SpawnBurst_Instantaneous(Use Spawn Probability=false, UseLoopCountLimit=true)
  ]
  particleSpawnScript [
   &InitializeParticle(
    Color=<rgba 0 3.3 9 0.05>
    Lifetime=5
    Mesh Scale Mode=::NewEnumerator3
    Mesh Uniform Scale=3
    UsePositionOffset=false
    Write Material Random=true
   )
   &CameraOffset(Camera Offset Amount=-25)
  ]
  particleUpdateScript [
   &ParticleState(Kill Particles When Lifetime Has Elapsed=false)
   &ScaleColor(
    Scale RGB=&VectorFromFloat(Value=&Waveform(
      Waveform Count=::NewEnumerator1
      "Waveform [1]"=::NewEnumerator0
      "[1] Frequency"=0.2
      "[1] Amplitude Min/Max"=<0.1 0.5>
      "[1] Bias"=0.25
      "[2] Average"=false
      "[2] Input Position"=0
      "[2] Phase"=$Emitter.Age
      "[2] Phase Scale"=1
      "[2] Compound Frequency"=2
      "[2] Frequency"=1
      "[2] Amplitude Min/Max"=<0.1 0.5>
      "[2] Scale"=1
      "[2] Bias"=0.5
      "[2] Exponent"=1
      "[2] Pulse Duration"=0.1
      "[2] Clip Triangle Min/Max"=<0 0>
      "[2] Random Seed"=0
      "Waveform [3] Blend"=::NewEnumerator5
     ))
    ScaleA=false
    ScaleRGB=true
   )
   &ScaleColor(
    Scale Alpha=&FloatFromCurve(
     FloatCurve=DataInterface() {propertyValues {Curve {
        defaultValue=0
        keys [
         {interpMode=RCIM_Cubic, tangentMode=RCTM_User, time=0, value=0.5}
         {interpMode=RCIM_Cubic, tangentMode=RCTM_User, time=0.5, value=1}
         {tangentMode=RCTM_User, time=1, value=0.5}
        ]
       }}}
     CurveIndex=Linked() {linkedVariable {name=NPC.FortniteNPC.TimeOfDay_0-1, type=NiagaraFloat}}
    )
    ScaleA=true
    ScaleRGB=false
   ) {enabled=false, module=ScaleColor001}
   &Light_Attributes(
    Color=1
    Color Value To Scale:LinearColor=$Particles.Color
    Scale Alpha:NiagaraFloat=&FloatFromCurve(
     FloatCurve=DataInterface() {propertyValues {Curve {
        defaultValue=0
        keys [
         {interpMode=RCIM_Cubic, tangentMode=RCTM_User, time=0, value=0.25}
         {interpMode=RCIM_Cubic, tangentMode=RCTM_User, time=0.5, value=1}
         {tangentMode=RCTM_User, time=1, value=0.25}
        ]
       }}}
     Scale Curve=6
    )
    Scale RGB:Vector3f=&VectorFromFloat(Value=30)
    ScaleA=true
    ScaleRGB=true
   )
   &OffsetPosition(Position Offset=&MakeVector(
     X=&Cosine(Normalized Angle=$Emitter.Age, Period=2, Scale=2, Bias=0)
     Y=&Sine(Normalized Angle=$Emitter.Age, Period=2, Scale=2, Bias=0)
     Z=&Sine(Normalized Angle=$Emitter.Age, Period=1, Scale=1, Bias=0)
    ))
   &DynamicMaterialParameters(
    Index 0 Param 1=&FloatFromCurve(
     FloatCurve=DataInterface() {propertyValues {Curve {
        defaultValue=0
        keys [
         {interpMode=RCIM_Cubic, tangentMode=RCTM_User, time=0, value=0.5}
         {interpMode=RCIM_Cubic, tangentMode=RCTM_User, time=0.5, value=1}
         {tangentMode=RCTM_User, time=1, value=0.5}
        ]
       }}}
     CurveIndex=Linked() {linkedVariable {name=NPC.FortniteNPC.TimeOfDay_0-1, type=NiagaraFloat}}
    )
    Module.Index0Float0Write=true
    Module.Index0Float1Write=false
    Module.Index0Float2Write=false
    Module.Index0Float3Write=false
   )
  ]
  rendererClasses [&NiagaraMeshRendererProperties, &NiagaraLightRendererProperties]
  renderers [
   {rendererClass=&NiagaraMeshRendererProperties, rendererIndex=0}
   {rendererClass=&NiagaraMeshRendererProperties, rendererIndex=1}
   {rendererClass=&NiagaraMeshRendererProperties, rendererIndex=2}
   {rendererClass=&NiagaraLightRendererProperties, rendererIndex=3}
  ]
 }
 SphereLightning {
  emitterUpdateScript [
   &EmitterState(
    Inactive Response=::NewEnumerator1
    Life Cycle Mode=::NewEnumerator1
    Loop Duration=&RandomRangeFloat(Minimum=0.133, Maximum=1.333)
    Recalculate Duration Each Loop=true
    UseLoopDelay=false
   )
   &SpawnRate(Spawn Probability=0.333, SpawnRate=4, Use Spawn Probability=true)
   &SpawnBurst_Instantaneous(Spawn Probability=0.5, Use Spawn Probability=true, UseLoopCountLimit=false)
  ]
  particleSpawnScript [
   &InitializeParticle(
    Color=<rgba 0 0.6609 1 0.25>
    Lifetime Max=0.733
    Lifetime Min=0.466
    Lifetime Mode=::NewEnumerator1
    Mesh Scale=&SelectVectorFromArray(
     Array Sampling Mode=::NewEnumerator0
     Direct Array Index=?
     Vector Selection Array=DataInterface() {propertyValues {FloatData [<1 -1 1>, <1 1 -1>, <1 -1 -1>, <1 1 1>]}}
     Random Seed=0
    )
    Mesh Scale Mode=::NewEnumerator4
    UsePositionOffset=false
    Write Material Random=true
   )
   /SetVariables(Particles.GlobalWidth:NiagaraFloat=&RandomRangeFloat(Minimum=1.5, Maximum=2.5)) {isSetParameters=true}
   /SetVariables(Particles.GlobalRandom=0.5) {isSetParameters=true}
   /SetVariables(Particles.MeshIndex=&ModuloInt(A=&ReturnParticleIDAttributes(), B=6)) {isSetParameters=true}
  ]
  particleUpdateScript [
   &ParticleState()
   &UpdateMeshOrientation(
    Facing Direction=&RandomRangeVector(Minimum=<0 -1 -1>, Maximum=<1 1 1>)
    Facing Interpolation=true
    Facing Interpolation Rate=0.5
    Orientation Method=::NewEnumerator1
    Override Quat=false
    Side Direction:Vector3f=&ReturnCameraVector(Camera Vectors=::NewEnumerator1, Coordinate Space=::Simulation)
    Stabilization Mode:ENiagara_OrientVectorCount=::NewEnumerator2
    TransformFacing=true
    TransformSide=true
   )
   &DynamicMaterialParameters(
    Index 0 Param 1=&FloatFromCurve(
     FloatCurve=DataInterface() {propertyValues {Curve {
        defaultValue=0
        keys [
         {interpMode=RCIM_Cubic, tangentMode=RCTM_User, time=0, value=0}
         {arriveTangent=1.3983, interpMode=RCIM_Cubic, leaveTangent=1.3983, time=0.3529, value=0.4369}
         {arriveTangent=5.5065, interpMode=RCIM_Cubic, leaveTangent=5.5065, time=0.8933, value=1.249}
         {arriveTangent=16.4692, interpMode=RCIM_Cubic, leaveTangent=16.4692, tangentMode=RCTM_User, time=1, value=4}
        ]
       }}}
     Scale Curve=$Particles.GlobalWidth
    )
    Index 0 Param 2=1
    Index 0 Param 3=&RandomRangeFloat(Minimum=0.25, Maximum=1)
    Index 0 Param 4=&FloatFromCurve(FloatCurve=DataInterface() {propertyValues {Curve {
        defaultValue=0
        keys [
         {interpMode=RCIM_Cubic, tangentMode=RCTM_User, time=0, value=0}
         {arriveTangent=2.7163, interpMode=RCIM_Cubic, leaveTangent=2.7163, tangentMode=RCTM_User, time=0.7, value=1}
        ]
       }}})
    Index 1 Param 1=&FloatFromCurve(FloatCurve=DataInterface() {propertyValues {Curve {
        defaultValue=0.0353
        keys [{interpMode=RCIM_Cubic, tangentMode=RCTM_User, time=0.5, value=0}, {interpMode=RCIM_Cubic, tangentMode=RCTM_User, time=1, value=1}]
       }}})
    Index 1 Param 2=&Multiply_Float(A=1.35, B=$Particles.GlobalRandom)
    Index 1 Param 4=0.5
    Module.Index0Float0Write=true
    Module.Index0Float1Write=true
    Module.Index0Float2Write=true
    Module.Index0Float3Write=true
    Module.Index1Float0Write=true
    Module.Index1Float1Write=true
    Module.Index1Float2Write=true
    Module.Index1Float3Write=true
    Write Parameter Index 1=::NewEnumerator1
   )
   &ScaleColor(
    Scale Alpha=&FloatFromCurve(
     FloatCurve=DataInterface() {propertyValues {Curve {
        defaultValue=0
        keys [
         {interpMode=RCIM_Cubic, tangentMode=RCTM_User, time=0, value=0.5}
         {interpMode=RCIM_Cubic, tangentMode=RCTM_User, time=0.5, value=1}
         {tangentMode=RCTM_User, time=1, value=0.5}
        ]
       }}}
     CurveIndex=Linked() {linkedVariable {name=NPC.FortniteNPC.TimeOfDay_0-1, type=NiagaraFloat}}
    )
    ScaleA=true
    ScaleRGB=true
   ) {enabled=false, module=ScaleColor001}
   /SetVariables(
    Particles.LightPosition:NiagaraPosition=&AddVectorToPosition(
     Position=$Particles.Position
     Vector=&Multiply_VectorByFloat(
      Vector=&ReturnMeshOrientationAxis(Mesh Orientation Axis=::XAxis, Mesh Orientation Quat=$Particles.MeshOrientation)
      Float=9
     )
    )
    Particles.LightRadius=&FloatFromCurve(FloatCurve=DataInterface() {propertyValues {Curve {keys [<at 0 1 RCIM_Constant>,<at 0.5 0>]}}}, Scale Curve=100)
    Particles.Scale=&VectorFromFloat(Value=&FloatFromCurve(
      FloatCurve=DataInterface() {propertyValues {Curve {keys [{arriveTangent=4.1785, interpMode=RCIM_Cubic, leaveTangent=1.1626, tangentMode=RCTM_Break, time=0, value=0.666}, <at 1 1>]}}}
      Scale Curve=3
     ))
    Particles.SphereGlowColor:LinearColor=<rgba 0 0.8696 1 1>
    Particles.SphereGlowScale:Vector3f=&VectorFromFloat(Value=1.5)
   ) {isSetParameters=true}
  ]
  rendererClasses [&NiagaraMeshRendererProperties]
  renderers [
   {rendererClass=&NiagaraMeshRendererProperties, rendererIndex=0}
   {rendererClass=&NiagaraMeshRendererProperties, rendererIndex=1}
  ]
 }
}
summary {
 systemName=NS_FriedNexus_EnergyGate_FX
 emitters [
  {emitterName=SphereLightning, rendererClasses [&NiagaraMeshRendererProperties]}
  {emitterName=Sphere, rendererClasses [&NiagaraMeshRendererProperties, &NiagaraLightRendererProperties]}
  {emitterName=SideLightning, rendererClasses [&NiagaraMeshRendererProperties, &NiagaraSpriteRendererProperties]}
  {emitterName=AirglowAndDecals, rendererClasses [&NiagaraSpriteRendererProperties]}
 ]
 userVariables [
  User.bIsFrontEndPreview {type=NiagaraBool}
  User.bisInVehicle {type=NiagaraBool}
 ]
}
